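# Additive unlit Niagara sprite material that draws animated light rays for the Ponder emote.
# An inky smoke texture is sampled in panned radial coordinates, using explicit derivatives from the original UVs, then multiplied by a soft powered sphere mask and particle alpha.
# Particle color is hue-shifted by radial distance and fed through an emissive adjustment function.
# Dynamic parameters drive the pan time and hue shift amount.
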Material /Game/Animation/Game/MainPlayer/Emotes/Ponder/FX/M_Emote_Ponder_Lightrays
graph {
 properties {bUsedWithNiagaraSprites=true, blendMode=BLEND_Additive, pixelDepthOffsetMode=PDOM_Legacy, shadingModel=MSM_Unlit, usedShadingModels=Unlit}
 nodes [
  n1 = MaterialFunctionCall() {materialFunction=&VectorToRadialValue}
  n2 = DynamicParameter() {paramNames [RaysTime,Hueshift,Param3,Param4]}
  n3 = ParticleColor()
  MaterialFunctionCall(Emissive=MaterialFunctionCall(Hue Shift Percentage=Multiply(A=@n1.Linear Distance, B=@n2.Hueshift), Texture=@n3.RGB) {materialFunction=&HueShift}.Result) {materialFunction=&MF_FX_EmissiveAdjustments}
  n4 = Panner(Coordinate=Multiply(A=@n1.Radial Coordinates, B=Constant2Vector() {r=2.5}), Time=@n2.RaysTime) {bFractionalPart=true, speedY=-0.1}
  n5 = TextureCoordinate()
  n6 = DDX(Value=@n5)
  n7 = DDY(Value=@n5)
  n8 = SphereMask(A=TextureCoordinate(), B=Constant() {r=0.5}) {attenuationRadius=0.5, hardnessPercent=0}
  Multiply(
   A=Multiply(
    A=TextureSample!(UVs=@n4, Tex, "DDX(UVs)"=@n6, "DDY(UVs)"=@n7, Apply View MipBias) {mipValueMode=TMVM_Derivative, texture=/Game/Effects/Fort_Effects/Textures/Tile/T_Inky_Smoke_Tile}.G
    B=Power(Base=@n8)
   )
   B=@n3.A
  )
 ]
}
